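; Katana Dash Charge gameplay ability that ejects the player from open vehicle seats, ridden actors, ziplines, grind rails or surface swimming.
; Vehicle, grind and swim exits use a constant root-motion force from cached horizontal velocity plus upward speed, repeating while still swimming, up to a limit.
; Riding exits use LaunchCharacter; ziplines simply end.
; On end after launching, it uncrouches and non-authority starts the Mantis technique; otherwise it ends the current technique.
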(event Custom|LaunchPlayerUp
  (params (ZVelocity :type Float%20%28double-precision%29))
  (bind (_cachedvelocity_x _cachedvelocity_y) (Math|Vector|BreakVector (Variables|GetCachedVelocity_X)))
  (bind _launchcounter (Variables|Default|GetLaunchCounter))
  (Variables|Default|SetbHasLaunched true)
  (bind _playerref (Variables|Default|GetPlayerRef)
    (:then
      (Networking|ForceNetUpdate _playerref)
      (bind _asynctaskproxy (Ability|Tasks|ApplyRootMotionConstantForce "DashChargeEject_ConstantForce" (Math|Vector|Normalize (Math|Vector|MakeVector _cachedvelocity_x _cachedvelocity_y ZVelocity)) (Math|Vector|VectorLength (Math|Vector|MakeVector _cachedvelocity_x _cachedvelocity_y ZVelocity)) 0.1)
        (:then
          (Variables|Default|SetDashChargeEjectConstantForce _asynctaskproxy))
        (:OnFinish
          (bind _playerref_2 (Variables|Default|GetPlayerRef)
            (:then
              (Networking|ForceNetUpdate _playerref_2)
              (if (GameplayTags|HasMatchingGameplayTag (Variables|Default|GetPlayerRef) "(TagName=\"MovementMode.SurfaceSwimming\")")
                (Math|Integer|IncrementInt _launchcounter
                  (:%20%20
                    (if (<= _launchcounter (ScalableFloat|GetValueatLevel (Variables|Default|GetMaxNumLaunches)))
                      (CallFunction|LaunchPlayerUp (ScalableFloat|GetValueatLevel (Variables|Default|GetJumpOutOfWaterZVelocity)))
                      (else
                        _))))
                (else
                  _))
              (Ability|EndAbility))
            (:else
              _)))))
    (:else
      (Ability|EndAbility))))

(event Ability|EventOnEndAbility
  (params (bWasCancelled :type Boolean))
  (bind _playerref (Variables|Default|GetPlayerRef))
  (|CallParentFunction bWasCancelled)
  (Utilities|FlowControl|Sequence
    (:then_0
      (bind _dashchargeejectconstantforce (Variables|Default|GetDashChargeEjectConstantForce)
        (:then
          (GameplayTasks|EndTask _dashchargeejectconstantforce))
        (:else)))
    (:then_1
      (bind _playerref_2 (Variables|Default|GetPlayerRef)
        (:then
          (Utilities|IsValid (Actor|GetComponentbyClass _playerref_2 "/Script/MantisRuntime.FortMantisPawnComponent")
            (:Is%20Valid
              (if (Variables|Default|GetbHasLaunched)
                (Character|UnCrouch _playerref)
                (Ability|HasAuthority
                  (:False
                    (Ability|GameplayEffect|AddLooseGameplayTags _playerref (Variables|Default|GetBlockEquipAnimTag))
                    (Mantis|SetBranchRule (Actor|GetComponentbyClass _playerref_2 "/Script/MantisRuntime.FortMantisPawnComponent") "Any" "Default" false)
                    (bind _returnvalue (Mantis|TryManuallyStartTechnique (Actor|GetComponentbyClass _playerref_2 "/Script/MantisRuntime.FortMantisPawnComponent") 1))
                    (if (not _returnvalue)
                      (Mantis|TryManuallyEndCurrentTechnique (Actor|GetComponentbyClass _playerref_2 "/Script/MantisRuntime.FortMantisPawnComponent"))
                      (Ability|GameplayEffect|RemoveLooseGameplayTags (Variables|Default|GetPlayerRef) (Variables|Default|GetBlockEquipAnimTag)))))
                (else
                  (Mantis|TryManuallyEndCurrentTechnique (Actor|GetComponentbyClass _playerref_2 "/Script/MantisRuntime.FortMantisPawnComponent")))))))
        (:else)))))

(event Ability|EventActivateAbility
  (bind _playerref (Variables|Default|GetPlayerRef))
  (bind _returnvalue (Actor|GetComponentbyClass _playerref "/Script/RidingCodeRuntime.RiderComponent"))
  (bind _donotuse_activeridable (Variables|/Script/RidingCodeRuntime.RiderComponent|GetDoNotUse_ActiveRidable _returnvalue))
  (bind _returnvalue_2 (Components|GetOwner _donotuse_activeridable))
  (bind _returnvalue_3 (Actor|GetComponentbyClass _returnvalue_2 "/Script/Engine.CharacterMovementComponent"))
  (|CallParentFunction)
  (bind _outfortplayerpawn (CallFunction|GetAvatarasValidFortPlayerPawn)
    (:ValidFortPlayerPawn
      (Variables|Default|SetPlayerRef _outfortplayerpawn)
      (Variables|Default|GetPlayerRef
        (:then
          (Networking|ForceNetUpdate (Variables|Default|GetPlayerRef))
          (if (GameplayTags|HasMatchingGameplayTag (Variables|Default|GetPlayerRef) "(TagName=\"Gameplay.Status.InVehicle.Seat.Open\")")
            ;[1 1 1] Open seated vehicle
            (Variables|Default|SetCachedVelocity (Transformation|GetVelocity (Vehicle|GetVehicle (Variables|Default|GetPlayerRef)))) ; CacheVehicleVelocity
            (Vehicle|EjectPlayerPawn (Vehicle|GetVehicle (Variables|Default|GetPlayerRef)) (Variables|Default|GetPlayerRef))
            (Ability|Tasks|WaitGameplayTagQueryOnActor (Variables|Default|GetPlayerRef) (Variables|Default|GetEjectQuery) "WhenTrue" true
              (:Triggered
                (CallFunction|LaunchPlayerUp (ScalableFloat|GetValueatLevel (Variables|Default|GetJumpOffVehicleZVelocity)))))
            (elif (GameplayTags|HasMatchingGameplayTag (Variables|Default|GetPlayerRef) "(TagName=\"Granted.Riding.Player.IsRiding\")")
              ;[1 1 1] Riding
              (Utilities|IsValid _returnvalue
                (:Is%20Valid
                  (Utilities|IsValid _donotuse_activeridable
                    (:Is%20Valid
                      (Utilities|IsValid _returnvalue_2
                        (:Is%20Valid
                          (Utilities|IsValid _returnvalue_3
                            (:Is%20Valid
                              (Variables|Default|SetCachedVelocity (Pawn|Components|CharacterMovement|GetLastUpdateVelocity _returnvalue_3))
                              (Riding|StopRidingActor _returnvalue false false false true)
                              (Ability|Tasks|WaitGameplayTagQueryOnActor (Variables|Default|GetPlayerRef) (Variables|Default|GetEjectQuery) "WhenTrue" true
                                (:Triggered
                                  (Variables|Default|SetbHasLaunched true)
                                  (Character|LaunchCharacter (Variables|Default|GetPlayerRef) (Math|Vector|MakeVector (.x (Variables|Default|GetCachedVelocity)) (.y (Variables|Default|GetCachedVelocity)) (ScalableFloat|GetValueatLevel (Variables|Default|GetJumpOffVehicleZVelocity))) false true)
                                  (Utilities|FlowControl|Delay 0.2)
                                  (Ability|EndAbility))))
                            (:Is%20Not%20Valid
                              _)))
                        (:Is%20Not%20Valid
                          _))
                      (Ability|EndAbility))
                    (:Is%20Not%20Valid
                      _)))
                (:Is%20Not%20Valid
                  _))
              (elif (or (GameplayTags|HasMatchingGameplayTag (Variables|Default|GetPlayerRef) "(TagName=\"MovementMode.ZipliningOnSpline\")") (GameplayTags|HasMatchingGameplayTag (Variables|Default|GetPlayerRef) "(TagName=\"MovementMode.ZipliningOnSpline\")"))
                ;[1 1 1] Ziplining
                (Zipline|EndZiplining (Variables|Default|GetPlayerRef) false false)
                (Variables|Default|SetbHasLaunched true)
                (Utilities|FlowControl|Delay 0.25)
                (Ability|EndAbility)
                (elif (GameplayTags|HasMatchingGameplayTag (Variables|Default|GetPlayerRef) "(TagName=\"MovementMode.Grinding\")")
                  ;[1 1 1] Grinding
                  (Ability|GameplayEffect|AddLooseGameplayTags (Variables|Default|GetPlayerRef) (Variables|Default|GetBlockGrindRailsTag))
                  (Variables|Default|SetCachedVelocity (Pawn|Components|CharacterMovement|GetLastUpdateVelocity (Variables|/Script/Engine.Character|GetCharacterMovement (Variables|Default|GetPlayerRef))))
                  (Pawn|Components|CharacterMovement|SetMovementMode (Variables|/Script/Engine.Character|GetCharacterMovement (Variables|Default|GetPlayerRef)) "MOVE_Walking" 0)
                  (CallFunction|LaunchPlayerUp (ScalableFloat|GetValueatLevel (Variables|Default|GetJumpOffVehicleZVelocity)))
                  (elif (GameplayTags|HasMatchingGameplayTag (Variables|Default|GetPlayerRef) "(TagName=\"MovementMode.SurfaceSwimming\")")
                    ;[1 1 1] Swimming
                    (Variables|Default|SetCachedVelocity (Pawn|Components|CharacterMovement|GetLastUpdateVelocity (Variables|/Script/Engine.Character|GetCharacterMovement (Variables|Default|GetPlayerRef))))
                    (CallFunction|LaunchPlayerUp (ScalableFloat|GetValueatLevel (Variables|Default|GetJumpOutOfWaterZVelocity)))
                    (else
                      (Ability|EndAbility))))))))
        (:else
          _)))
    (:AvatarCastFailed
      _))
  (Ability|EndAbility))

(event Ability|EventFailedToActivatePassiveAbility
  (params (ActorInfo :type Gameplay%20Ability%20Actor%20Info%20Structure)))

(fn CalculateVehicleLaunch
  (return "0, 0, 0" (Math|Vector|Normalize "0, 0, 0") (Math|Vector|VectorLength "0, 0, 0")))

(fn ResetMovementMode)
