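; Passive gameplay ability that grants a bonus dash charge to the katana melee weapon.
; On activation it gets the avatar as a valid player pawn, loops over that pawn's weapon list and casts each entry to B_Melee_Katana.
; Each katana gets one ammo added through ModAmmo, so the charge is represented as weapon ammo.
; The failed-activation and end-ability events only call their parents.
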(event Ability|EventFailedToActivatePassiveAbility
  (params (ActorInfo :type Gameplay%20Ability%20Actor%20Info%20Structure))
  (|CallParentFunction ActorInfo))

(event Ability|EventActivateAbility
  (|CallParentFunction)
  (bind _outfortplayerpawn (CallFunction|GetAvatarasValidFortPlayerPawn)
    (:ValidFortPlayerPawn
      (for _array_element (Variables|/Script/FortniteGame.FortPawn|GetCurrentWeaponList _outfortplayerpawn)
        (bind _asb_melee_katana (Utilities|Casting|CastToB_Melee_Katana _array_element))
        (FortWeapon|ModAmmo _asb_melee_katana 1)))))

(event Ability|EventOnEndAbility
  (params (bWasCancelled :type Boolean))
  (|CallParentFunction bWasCancelled))
